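# Sound cue for the charged state of the electric wall trap, playing an endless loop of randomly picked spark sounds reused from the ceiling electric trap.
# A quality-level root selects a looping random pool of four sparks at the first level or three at the second, with the other two slots empty.
# It uses the Fort_Traps sound class and attenuation, with overridden concurrency that prevents new instances once the limit is reached.
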SoundCue /Figment_S07/Sounds/Fort_Traps/Wall_Electric/Wall_Electric_Trap_Charged
root=QualityLevel_0
nodes [
 {id=Looping_1, children [Random_1]}
 {id=Looping_2, children [Random_2]}
 {id=QualityLevel_0, children [Looping_1,Looping_2,null,null]}
 {id=Random_1, children [WavePlayer_4,WavePlayer_5,WavePlayer_6,WavePlayer_7], properties {weights [1,1,1,1]}}
 {id=Random_2, children [WavePlayer_8, WavePlayer_9, WavePlayer_11], properties {weights [1,1,1]}}
 {id=WavePlayer_11, properties {soundWaveAssetPtr=/Game/Sounds/Fort_Traps/Electric_Trap/CeilingTrap_Electric_Spark_002}}
 {id=WavePlayer_4, properties {soundWaveAssetPtr=/Game/Sounds/Fort_Traps/Electric_Trap/CeilingTrap_Electric_Spark_008}}
 {id=WavePlayer_5, properties {soundWaveAssetPtr=/Game/Sounds/Fort_Traps/Electric_Trap/CeilingTrap_Electric_Spark_007}}
 {id=WavePlayer_6, properties {soundWaveAssetPtr=/Game/Sounds/Fort_Traps/Electric_Trap/CeilingTrap_Electric_Spark_004}}
 {id=WavePlayer_7, properties {soundWaveAssetPtr=&CeilingTrap_Electric_Spark_002}}
 {id=WavePlayer_8, properties {soundWaveAssetPtr=&CeilingTrap_Electric_Spark_008}}
 {id=WavePlayer_9, properties {soundWaveAssetPtr=&CeilingTrap_Electric_Spark_007}}
]
properties {
 attenuationSettings=&Fort_Traps_Att
 bOverrideConcurrency=true
 maxDistance=6000
 soundClassObject=&Fort_Traps
 volumeMultiplier=0.5
 attenuationOverrides {falloffDistance=2100}
 concurrencyOverrides {resolutionRule=PreventNew}
}
